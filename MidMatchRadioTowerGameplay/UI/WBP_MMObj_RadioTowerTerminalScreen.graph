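# Screen widget for the mid-match Radio Tower terminal, showing status text such as "Acquiring Forecast" or an "Access Denied" error.
# A bordered overlay layers dimmed spinner rings behind a widget switcher holding the status page and the error page.
# Custom events switch the error page, update primary and secondary text, and toggle the initially collapsed secondary line.
# PlayActivateAnimation fades in the whole content, with the switcher fading in slightly later.
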WidgetBlueprint /MidMatchRadioTowerGameplay/UI/WBP_MMObj_RadioTowerTerminalScreen
animations [Activate {
  binding_count=2
  length_seconds=1.25
  object=&self:Activate
  sequence {bindings [
    Overlay_Content {
     id=#1
     possessed_class=Overlay
     tracks [{
       type=MovieSceneFloatTrack
       display_name=Render Opacity
       sections [{type=MovieSceneFloatSection, end_frame=20, start_frame=0, channels [None {type=MovieSceneScriptingFloatChannel, keys [[0,0],[20,1]]}]}]
      }]
    }
    CommonWidgetSwitcher {
     id=#2
     possessed_class=CommonWidgetSwitcherLegacy
     tracks [{
       type=MovieSceneFloatTrack
       display_name=Render Opacity
       sections [{type=MovieSceneFloatSection, end_frame=25, start_frame=0, channels [None {type=MovieSceneScriptingFloatChannel, keys [[0,0],[10,0],[25,1]]}]}]
      }]
    }
   ]}
 }]
graphs [EventGraph {nodes [
   n1 = SetEnableErrorScreen!() -> (OutputDelegate, then, bEnable:bool=~) {is=title K2Node_CustomEvent}
   > n2 = Branch(Condition=@n1.bEnable)
   > SetActiveWidgetIndex(self=Get CommonWidgetSwitcher() -> (CommonWidgetSwitcher:Common Widget Switcher Legacy Object Reference), Index=1)
   SetActiveWidgetIndex(execute=@n2.else, self=Get CommonWidgetSwitcher(), Index=0)
   n3 = SetEnableSecondaryText!() -> (OutputDelegate, then, bEnable:bool=~) {is=title K2Node_CustomEvent}
   > n4 = Get!(execute:Exec) -> (then:Exec, else:Exec, SecondaryText:Common Text Object Reference)
   > SetVisibility(self=@n4, InVisibility=Select(Option 0=Collapsed, Option 1=HitTestInvisible, Index=@n3.bEnable))
   n5 = UpdatePrimaryText!() -> (OutputDelegate, then, InText:"Text (by ref)") {is=title K2Node_CustomEvent}
   > n6 = Get!(execute) -> (then, else, Text:Common Text Object Reference)
   > "SetText (Text)"(self=@n6, InText=@n5.InText)
   n7 = UpdateSecondaryText!() -> (OutputDelegate, then, InText:Text) {is=title K2Node_CustomEvent}
   > n8 = Get!(execute) -> (then, else, SecondaryText:Common Text Object Reference)
   > "SetText (Text)"(self=@n8, InText=@n7.InText)
   > InvalidateLayoutAndVolatility(self:Widget Object Reference=@n8)
   PlayActivateAnimation() {is=title K2Node_CustomEvent}
   > PlayAnimation(InAnimation=Get Activate() -> (Activate:Widget Animation Object Reference))
  ]}]
variables [ErrorScreenDurationTimerHandle {type=TimerHandle}]
widgets {
 info {inheritedWidgetCount=0, parentClass=/Script/FortniteUI.FortUserWidget, rootWidgetClass=&Overlay, widgetCount=11}
 widgets [
  Overlay_Content {widgetClassPath=&Overlay, widget=Overlay_Content}
  Image_ScreenBorder {
   widgetClassPath=&Image
   bIsVariable=true
   parent=Overlay_Content
   widget=Image_ScreenBorder
   properties {brush {
     bIsDynamicallyLoaded=false
     drawAs=Box
     imageType=NoImage
     mirroring=NoMirror
     resourceName=None
     resourceObject=&/M_UI_TargetBracket_RadioTower
     tiling=NoTile
     imageSize=<64 64>
     margin=<ltrb 0.4 0.4 0.4 0.4>
     outlineSettings {
      bUseBrushTransparency=false
      roundingType=HalfHeightRadius
      width=0
      color {colorUseRule=UseColor_Specified, specifiedColor=<rgba 0 0 0 0>}
      cornerRadii=<0 0 0 0>
     }
     tintColor {colorUseRule=UseColor_Specified, specifiedColor=<rgba 1 1 1 1>}
     uVRegion {bIsValid=false, max=<0 0>, min=<0 0>}
    }}
   slot {class=OverlaySlot, properties {horizontalAlignment=HAlign_Fill, verticalAlignment=VAlign_Fill, padding=<ltrb 5 5 5 5>}}
  }
  CommonVisualAttachment_0 {
   widgetClassPath=&CommonVisualAttachment
   parent=Overlay_Content
   widget=CommonVisualAttachment_0
   properties {renderOpacity=0.4, contentAnchor=<0.5 0.5>}
   slot {class=OverlaySlot, properties {horizontalAlignment=HAlign_Center, verticalAlignment=VAlign_Center}}
  }
  Overlay_BGIdle {widgetClassPath=&Overlay, bIsVariable=true, parent=CommonVisualAttachment_0, widget=Overlay_BGIdle, slot {class=SizeBoxSlot}}
  Image_DashRing {
   widgetClassPath=&Image
   bIsVariable=true
   parent=Overlay_BGIdle
   widget=Image_DashRing
   properties {brush {
     bIsDynamicallyLoaded=false
     drawAs=Image
     imageType=NoImage
     mirroring=NoMirror
     resourceName=None
     resourceObject=/Game/Athena/Items/EnvironmentalItems/Locks/Scanner/Materials/M_UI_Spinner_Inst
     tiling=NoTile
     imageSize=<150 150>
     margin=<ltrb 0 0 0 0>
     outlineSettings {
      bUseBrushTransparency=false
      roundingType=HalfHeightRadius
      width=0
      color {colorUseRule=UseColor_Specified, specifiedColor=<rgba 0 0 0 0>}
      cornerRadii=<0 0 0 0>
     }
     tintColor {colorUseRule=UseColor_Specified, specifiedColor=<rgba 1 1 1 1>}
     uVRegion {bIsValid=false, max=<0 0>, min=<0 0>}
    }}
   slot {class=OverlaySlot, properties {horizontalAlignment=HAlign_Center, verticalAlignment=VAlign_Center}}
  }
  Image_DotRing {
   widgetClassPath=&Image
   bIsVariable=true
   parent=Overlay_BGIdle
   widget=Image_DotRing
   properties {brush {
     bIsDynamicallyLoaded=false
     drawAs=Image
     imageType=NoImage
     mirroring=NoMirror
     resourceName=None
     resourceObject=/Game/Athena/Items/EnvironmentalItems/Locks/Scanner/Materials/M_UI_Spinner_Inst1
     tiling=NoTile
     imageSize=<150 150>
     margin=<ltrb 0 0 0 0>
     outlineSettings {
      bUseBrushTransparency=false
      roundingType=HalfHeightRadius
      width=0
      color {colorUseRule=UseColor_Specified, specifiedColor=<rgba 0 0 0 0>}
      cornerRadii=<0 0 0 0>
     }
     tintColor {colorUseRule=UseColor_Specified, specifiedColor=<rgba 1 1 1 1>}
     uVRegion {bIsValid=false, max=<0 0>, min=<0 0>}
    }}
   slot {class=OverlaySlot, properties {horizontalAlignment=HAlign_Center, verticalAlignment=VAlign_Center}}
  }
  CommonWidgetSwitcher {
   widgetClassPath=&CommonWidgetSwitcherLegacy
   bIsVariable=true
   parent=Overlay_Content
   widget=CommonWidgetSwitcher
   slot {class=OverlaySlot, properties {horizontalAlignment=HAlign_Fill, verticalAlignment=VAlign_Fill}}
  }
  VerticalBox_0 {
   widgetClassPath=&VerticalBox
   parent=CommonWidgetSwitcher
   widget=VerticalBox_0
   slot {class=WidgetSwitcherSlot, properties {horizontalAlignment=HAlign_Center, verticalAlignment=VAlign_Center}}
  }
  Text {
   widgetClassPath=&CommonTextBlock
   bIsVariable=true
   parent=VerticalBox_0
   widget=Text
   properties {
    justification=Center
    lineHeightPercentage=0.8
    text=Acquiring Forecast
    textTransformPolicy=ToUpper
    wrapTextAt=150
    font {
     bForceMonospaced=false
     bMaterialIsStencil=false
     fontMaterial=null
     fontObject=&Burbank
     letterSpacing=0
     monospacedWidth=1
     size=13.5
     skewAmount=0
     typefaceFontName=BurbankRegular-Black
     outlineSettings {
      bApplyOutlineToDropShadows=false
      bMiteredCorners=false
      bSeparateFillAlpha=false
      outlineMaterial=null
      outlineSize=0
      outlineColor=<rgba 0 0 0 1>
     }
    }
   }
   slot {class=VerticalBoxSlot}
  }
  SecondaryText {
   widgetClassPath=&CommonTextBlock
   bIsVariable=true
   parent=VerticalBox_0
   widget=SecondaryText
   properties {
    justification=Center
    lineHeightPercentage=0.8
    text=10 %
    textTransformPolicy=ToUpper
    visibility=Collapsed
    font {
     bForceMonospaced=false
     bMaterialIsStencil=false
     fontMaterial=null
     fontObject=&Burbank
     letterSpacing=0
     monospacedWidth=1
     size=18
     skewAmount=0
     typefaceFontName=BurbankRegular-Black
     outlineSettings {
      bApplyOutlineToDropShadows=false
      bMiteredCorners=false
      bSeparateFillAlpha=false
      outlineMaterial=null
      outlineSize=0
      outlineColor=<rgba 0 0 0 1>
     }
    }
    shadowOffset=<0 0>
   }
   slot {class=VerticalBoxSlot}
  }
  Text_Error {
   widgetClassPath=&CommonTextBlock
   parent=CommonWidgetSwitcher
   widget=Text_Error
   properties {
    justification=Center
    lineHeightPercentage=0.8
    text="Access\r\nDenied"
    textTransformPolicy=ToUpper
    font {
     bForceMonospaced=false
     bMaterialIsStencil=false
     fontMaterial=null
     fontObject=&Burbank
     letterSpacing=0
     monospacedWidth=1
     size=13.5
     skewAmount=0
     typefaceFontName=BurbankRegular-Black
     outlineSettings {
      bApplyOutlineToDropShadows=false
      bMiteredCorners=false
      bSeparateFillAlpha=false
      outlineMaterial=null
      outlineSize=0
      outlineColor=<rgba 0 0 0 1>
     }
    }
    shadowOffset=<0 0>
   }
   slot {class=WidgetSwitcherSlot, properties {horizontalAlignment=HAlign_Center, verticalAlignment=VAlign_Center}}
  }
 ]
}
